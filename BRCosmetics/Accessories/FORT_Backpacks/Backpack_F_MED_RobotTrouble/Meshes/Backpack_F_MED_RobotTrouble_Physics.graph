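# Physics asset for the RobotTrouble backpack cosmetic, defining capsule collision bodies and joints for its dangling cloth and rope bones.
# Two capsules form the backpack root body, which anchors four four-bone cloth chains (left and right pairs) and two rope chains.
# Each link constrains to its parent with locked twist and limited swing.
# The cloth chains are stiffer at the root joint and looser further down.
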PhysicsAsset /BRCosmetics/Accessories/FORT_Backpacks/Backpack_F_MED_RobotTrouble/Meshes/Backpack_F_MED_RobotTrouble_Physics
bodies {
 backpack_root [
  {shapeType=Capsule, length=5.8569, radius=9.34, center=<-3.7342 -4.7839 -0.1429>, rotation=<rot -2.2119 -176.5297 179.8663>}
  {shapeType=Capsule, length=8.5636, radius=5.64, center=<-13.2613 -5.4838 0>}
 ]
 cloth_l1_1 [{shapeType=Capsule, length=1.6494, radius=0.6974, rotation=<rot -85.7478 0.0001 -0.0001>}]
 cloth_l1_2 [{shapeType=Capsule, length=1.6494, radius=0.6974, rotation=<rot -85.7479 0 0>}]
 cloth_l1_3 [{shapeType=Capsule, length=1.0374, radius=1.3094, rotation=<rot -85.7482 0 0>}]
 cloth_l1_4 [{shapeType=Capsule, length=1.0374, radius=1.3094, rotation=<rot -85.7477 0 0>}]
 cloth_l2_1 [{shapeType=Capsule, length=1.6494, radius=0.6974, rotation=<rot -85.7478 0.0001 -0.0001>}]
 cloth_l2_2 [{shapeType=Capsule, length=1.6494, radius=0.6974, rotation=<rot -85.7479 0 0>}]
 cloth_l2_3 [{shapeType=Capsule, length=1.0374, radius=1.3094, rotation=<rot -85.7482 0 0>}]
 cloth_l2_4 [{shapeType=Capsule, length=1.0374, radius=1.3094, rotation=<rot -85.7477 0 0>}]
 cloth_r1_1 [{shapeType=Capsule, length=1.6494, radius=0.6974, rotation=<rot -85.7477 0 0>}]
 cloth_r1_2 [{shapeType=Capsule, length=1.6494, radius=0.6974, rotation=<rot -85.7482 0 0>}]
 cloth_r1_3 [{shapeType=Capsule, length=1.0374, radius=1.3094, rotation=<rot -85.7482 0 0>}]
 cloth_r1_4 [{shapeType=Capsule, length=1.0374, radius=1.3094, rotation=<rot -85.7484 0 0>}]
 cloth_r2_1 [{shapeType=Capsule, length=1.6494, radius=0.6974, rotation=<rot -85.7477 0 0>}]
 cloth_r2_2 [{shapeType=Capsule, length=1.6494, radius=0.6974, rotation=<rot -85.7482 0 0>}]
 cloth_r2_3 [{shapeType=Capsule, length=1.0374, radius=1.3094, rotation=<rot -85.7482 0 0>}]
 cloth_r2_4 [{shapeType=Capsule, length=1.0374, radius=1.3094, rotation=<rot -85.7484 0 0>}]
 rope1_1 [{shapeType=Capsule, length=2.9254, radius=0.6484, center=<0.2623 0 0.0195>, rotation=<rot -89.3442 -3.6035 4.2537>}]
 rope1_3 [{shapeType=Capsule, length=2.829, radius=0.6484, center=<0.1551 0 0.0115>, rotation=<rot -89.3487 -3.6117 4.2619>}]
 rope1_5 [{shapeType=Capsule, length=3.6435, radius=1.7149, center=<1.8976 0 0.8463>, rotation=<rot -85.7484 -0.0001 0.0001>}]
 rope2_1 [{shapeType=Capsule, length=1.8467, radius=0.5, rotation=<rot -85.7487 0 0>}]
 rope2_2 [{shapeType=Capsule, length=0, radius=2.0016, center=<2.0517 -0.0001 0.4685>, rotation=<rot -85.7487 0 0>}]
}
constraints [
 {bone1Name=rope1_1, bone2Name=backpack_root, swing1LimitDegrees=15, swing2LimitDegrees=10, twistLimitDegrees=45, twistMotion=Locked}
 {bone1Name=rope1_3, bone2Name=rope1_1, swing1LimitDegrees=15, swing2LimitDegrees=10, twistLimitDegrees=45, twistMotion=Locked}
 {bone1Name=rope1_5, bone2Name=rope1_3, swing1LimitDegrees=15, swing2LimitDegrees=8, twistLimitDegrees=45, twistMotion=Locked}
 {bone1Name=rope2_1, bone2Name=backpack_root, swing1LimitDegrees=45, swing2LimitDegrees=45, twistLimitDegrees=45, twistMotion=Locked}
 {bone1Name=rope2_2, bone2Name=rope2_1, swing1LimitDegrees=45, swing2LimitDegrees=45, twistLimitDegrees=45, twistMotion=Locked}
 {bone1Name=cloth_r1_1, bone2Name=backpack_root, swing1LimitDegrees=5, swing2LimitDegrees=15, twistLimitDegrees=45, twistMotion=Locked}
 {bone1Name=cloth_r1_2, bone2Name=cloth_r1_1, swing1LimitDegrees=20, swing2LimitDegrees=45, twistLimitDegrees=45, twistMotion=Locked}
 {bone1Name=cloth_r1_3, bone2Name=cloth_r1_2, swing1LimitDegrees=20, swing2LimitDegrees=45, twistLimitDegrees=45, twistMotion=Locked}
 {bone1Name=cloth_r1_4, bone2Name=cloth_r1_3, swing1LimitDegrees=20, swing2LimitDegrees=45, twistLimitDegrees=45, twistMotion=Locked}
 {bone1Name=cloth_r2_1, bone2Name=backpack_root, swing1LimitDegrees=5, swing2LimitDegrees=15, twistLimitDegrees=45, twistMotion=Locked}
 {bone1Name=cloth_r2_2, bone2Name=cloth_r2_1, swing1LimitDegrees=20, swing2LimitDegrees=45, twistLimitDegrees=45, twistMotion=Locked}
 {bone1Name=cloth_r2_3, bone2Name=cloth_r2_2, swing1LimitDegrees=20, swing2LimitDegrees=45, twistLimitDegrees=45, twistMotion=Locked}
 {bone1Name=cloth_r2_4, bone2Name=cloth_r2_3, swing1LimitDegrees=20, swing2LimitDegrees=45, twistLimitDegrees=45, twistMotion=Locked}
 {bone1Name=cloth_l1_1, bone2Name=backpack_root, swing1LimitDegrees=5, swing2LimitDegrees=15, twistLimitDegrees=45, twistMotion=Locked}
 {bone1Name=cloth_l1_2, bone2Name=cloth_l1_1, swing1LimitDegrees=20, swing2LimitDegrees=45, twistLimitDegrees=45, twistMotion=Locked}
 {bone1Name=cloth_l1_3, bone2Name=cloth_l1_2, swing1LimitDegrees=20, swing2LimitDegrees=45, twistLimitDegrees=45, twistMotion=Locked}
 {bone1Name=cloth_l1_4, bone2Name=cloth_l1_3, swing1LimitDegrees=20, swing2LimitDegrees=45, twistLimitDegrees=45, twistMotion=Locked}
 {bone1Name=cloth_l2_1, bone2Name=backpack_root, swing1LimitDegrees=5, swing2LimitDegrees=15, twistLimitDegrees=45, twistMotion=Locked}
 {bone1Name=cloth_l2_2, bone2Name=cloth_l2_1, swing1LimitDegrees=20, swing2LimitDegrees=45, twistLimitDegrees=45, twistMotion=Locked}
 {bone1Name=cloth_l2_3, bone2Name=cloth_l2_2, swing1LimitDegrees=20, swing2LimitDegrees=45, twistLimitDegrees=45, twistMotion=Locked}
 {bone1Name=cloth_l2_4, bone2Name=cloth_l2_3, swing1LimitDegrees=20, swing2LimitDegrees=45, twistLimitDegrees=45, twistMotion=Locked}
]
